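# Anim Blueprint for the F_MED_Assemble_R female body that corrects shoulder pad and arm dyn bones on a pose copied from the attached parent mesh.
# Each update derives alphas from clavicle and upper-arm rotation, driving pad constraints and additive bone offsets, with extra pad shifts while skydiving.
# Arm constraints feed a Control Rig, followed by arm-aim offsets.
# Several dyn_arm_aim MSTR constraint and transform nodes are disconnected from the output.
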AnimBlueprint /Game/Characters/Player/Female/Medium/Bodies/F_MED_Assemble_R/Meshes/F_MED_Assemble_R_AnimBP
target_skeleton=&/F_MED_Assemble_R_Skeleton
graphs [
 AnimGraph {nodes [
   Component To Local()
   Component To Local()
   Component To Local()
   Component To Local()
   Component To Local(ComponentPose=Local To Component())
   Component To Local()
   Component To Local()
   Component To Local()
   n1 = "Constraint - dyn_arm_aim_r_MSTR (PP)"!(ComponentPose, Alpha=1, ConstraintWeights_0=0.8, ConstraintWeights_1:float=0.2) -> (Pose) {
    alphaBoolBlend {blendOption=Linear}
    boneToModify {boneName=dyn_arm_aim_r_MSTR}
    constraintSetup [{targetBone {boneName=upperarm_twist_01_r}}, {targetBone {boneName=upperarm_r}}]
    constraintWeights [1,1]
   }
   "Constraint - dyn_arm_aim_l_MSTR (PP)"!(ComponentPose=@n1, Alpha=1, ConstraintWeights_0=0.2, ConstraintWeights_1:float=0.8) -> (Pose) {
    alphaBoolBlend {blendOption=Linear}
    boneToModify {boneName=dyn_arm_aim_l_MSTR}
    constraintSetup [{targetBone {boneName=upperarm_l}}, {targetBone {boneName=upperarm_twist_01_l}}]
    constraintWeights [1,1]
   }
   Local To Component()
   Local To Component()
   "Transform (Modify) Bone - Bone: dyn_arm_aim_l_MSTR"(Alpha=1, Translation="0,0,0", Rotation="0,0,0", Scale="1,1,1") {
    rotationMode=BMM_Additive
    rotationSpace=BCS_BoneSpace
    translationMode=BMM_Additive
    translationSpace=BCS_BoneSpace
    alphaBoolBlend {blendOption=Linear}
    boneToModify {boneName=dyn_arm_aim_l_MSTR}
   }
   "Transform (Modify) Bone - Bone: dyn_arm_aim_r_MSTR"(Alpha=1, Translation="0,-4,3", Rotation="10,0,30", Scale="1,1,1") {
    rotationMode=BMM_Additive
    rotationSpace=BCS_BoneSpace
    scaleMode=BMM_Additive
    translationMode=BMM_Additive
    translationSpace=BCS_BoneSpace
    alphaBoolBlend {blendOption=Linear}
    boneToModify {boneName=dyn_arm_aim_r_MSTR}
    rotation {pitch=10, roll=30}
    translation {y=-4, z=3}
   }
   n2 = "Transform (Modify) Bone - Bone: dyn_arm_aim_l_MSTR"(Alpha=0, Translation="0,0,0", Rotation="0,0,0", Scale="1,1,1") {
    rotationMode=BMM_Additive
    rotationSpace=BCS_BoneSpace
    translationMode=BMM_Additive
    translationSpace=BCS_BoneSpace
    alphaBoolBlend {blendOption=Linear}
    boneToModify {boneName=dyn_arm_aim_l_MSTR}
   }
   "Transform (Modify) Bone - Bone: dyn_arm_aim_r_MSTR"(ComponentPose=@n2, Alpha=1, Translation="0,0,0", Rotation="0,0,0", Scale="1,1,1") {
    rotationMode=BMM_Additive
    rotationSpace=BCS_BoneSpace
    scaleMode=BMM_Additive
    translationMode=BMM_Additive
    translationSpace=BCS_BoneSpace
    alphaBoolBlend {blendOption=Linear}
    boneToModify {boneName=dyn_arm_aim_r_MSTR}
   }
   "Transform (Modify) Bone - Bone: dyn_arm_aim_r_MSTR"(Alpha=1, Translation="20,1,-1", Rotation="45,0,5", Scale="1,1,1") {
    rotationMode=BMM_Additive
    rotationSpace=BCS_BoneSpace
    scaleMode=BMM_Additive
    translationMode=BMM_Additive
    translationSpace=BCS_BoneSpace
    alphaBoolBlend {blendOption=Linear}
    boneToModify {boneName=dyn_arm_aim_r_MSTR}
    rotation {pitch=45, roll=5}
    translation=<20 1 -1>
   }
   n3 = "Transform (Modify) Bone - Bone: dyn_arm_aim_r_MSTR"(Alpha=1, Translation="0,0.5,-1.3", Rotation="0,100,0", Scale="1,1,1") {
    rotationMode=BMM_Additive
    rotationSpace=BCS_BoneSpace
    scaleMode=BMM_Additive
    translationMode=BMM_Additive
    translationSpace=BCS_BoneSpace
    alphaBoolBlend {blendOption=Linear}
    boneToModify {boneName=dyn_arm_aim_r_MSTR}
    rotation {yaw=100}
    translation {y=0.5, z=-1.3}
   }
   "Transform (Modify) Bone - Bone: dyn_arm_aim_r_MSTR"(ComponentPose=@n3, Alpha=1, Translation="-3,0,0", Rotation="0,0,0", Scale="1,1,1") {
    rotationMode=BMM_Additive
    rotationSpace=BCS_BoneSpace
    scaleMode=BMM_Additive
    translationMode=BMM_Additive
    translationSpace=BCS_BoneSpace
    alphaBoolBlend {blendOption=Linear}
    boneToModify {boneName=dyn_arm_aim_r_MSTR}
    translation {x=-3}
   }
   "Transform (Modify) Bone - Bone: dyn_arm_aim_r_MSTR"(Alpha=1, Translation="0,1,-1", Rotation="0,50,30", Scale="1,1,1") {
    rotationMode=BMM_Additive
    rotationSpace=BCS_BoneSpace
    scaleMode=BMM_Additive
    translationMode=BMM_Additive
    translationSpace=BCS_BoneSpace
    alphaBoolBlend {blendOption=Linear}
    boneToModify {boneName=dyn_arm_aim_r_MSTR}
    rotation {roll=30, yaw=50}
    translation {y=1, z=-1}
   }
   "Transform (Modify) Bone - Bone: dyn_arm_aim_r_MSTR"(Alpha=1, Translation="0,0.5,-1.3", Rotation="-10,100,0", Scale="1,1,1") {
    rotationMode=BMM_Additive
    rotationSpace=BCS_BoneSpace
    scaleMode=BMM_Additive
    translationMode=BMM_Additive
    translationSpace=BCS_BoneSpace
    alphaBoolBlend {blendOption=Linear}
    boneToModify {boneName=dyn_arm_aim_r_MSTR}
    rotation {pitch=-10, yaw=100}
    translation {y=0.5, z=-1.3}
   }
   "Transform (Modify) Bone - Bone: dyn_arm_aim_l_MSTR"(Alpha=0, Translation="0,0,0", Rotation="10,0,30", Scale="1,1,1") {
    rotationMode=BMM_Additive
    rotationSpace=BCS_BoneSpace
    translationMode=BMM_Additive
    translationSpace=BCS_BoneSpace
    alphaBoolBlend {blendOption=Linear}
    boneToModify {boneName=dyn_arm_aim_l_MSTR}
    rotation {pitch=10, roll=30}
   }
   n4 = "Transform (Modify) Bone - Bone: dyn_arm_aim_l_MSTR"(Alpha=1, Translation="0,0,0", Rotation="0,0,0", Scale="1,1,1") {
    rotationMode=BMM_Additive
    rotationSpace=BCS_BoneSpace
    translationMode=BMM_Additive
    translationSpace=BCS_BoneSpace
    alphaBoolBlend {blendOption=Linear}
    boneToModify {boneName=dyn_arm_aim_l_MSTR}
   }
   "Transform (Modify) Bone - Bone: dyn_arm_aim_r_MSTR"(ComponentPose=@n4, Alpha=0, Translation="0,0,0", Rotation="-30,30,0", Scale="1,1,1") {
    rotationMode=BMM_Additive
    rotationSpace=BCS_BoneSpace
    translationMode=BMM_Additive
    translationSpace=BCS_BoneSpace
    alphaBoolBlend {blendOption=Linear}
    boneToModify {boneName=dyn_arm_aim_r_MSTR}
    rotation {pitch=-30, yaw=30}
   }
   "Transform (Modify) Bone - Bone: dyn_arm_aim_r_MSTR"(Alpha=1, Translation="0,-1,-1", Rotation="10,0,-30", Scale="1,1,1") {
    rotationMode=BMM_Additive
    rotationSpace=BCS_BoneSpace
    scaleMode=BMM_Additive
    translationMode=BMM_Additive
    translationSpace=BCS_BoneSpace
    alphaBoolBlend {blendOption=Linear}
    boneToModify {boneName=dyn_arm_aim_r_MSTR}
    rotation {pitch=10, roll=-30}
    translation {y=-1, z=-1}
   }
   n5 = "Transform (Modify) Bone - Bone: dyn_arm_aim_l_MSTR"(ComponentPose=Local To Component(), Alpha=1, Translation="15,0,0", Rotation="0,0,0", Scale="1,1,1") {
    rotationMode=BMM_Additive
    rotationSpace=BCS_BoneSpace
    translationMode=BMM_Additive
    translationSpace=BCS_BoneSpace
    alphaBoolBlend {blendOption=Linear}
    boneToModify {boneName=dyn_arm_aim_l_MSTR}
    translation {x=15}
   }
   "Transform (Modify) Bone - Bone: dyn_arm_aim_r_MSTR"(ComponentPose=@n5, Alpha=1, Translation="15,0,0", Rotation="0,0,0", Scale="1,1,1") {
    rotationMode=BMM_Additive
    rotationSpace=BCS_BoneSpace
    scaleMode=BMM_Additive
    translationMode=BMM_Additive
    translationSpace=BCS_BoneSpace
    alphaBoolBlend {blendOption=Linear}
    boneToModify {boneName=dyn_arm_aim_r_MSTR}
    translation {x=15}
   }
   n6 = "Transform (Modify) Bone - Bone: dyn_shoulder_pad_l_MSTR"(
    ComponentPose=Local To Component(LocalPose=Copy Pose From Mesh() {bUseAttachedParent=true})
    Alpha=Get clav_up_plus() -> (clav_up_plus:double)
    Translation="0,-3,0"
    Rotation="0,0,0"
    Scale="1,1,1"
   ) {
    rotationMode=BMM_Additive
    rotationSpace=BCS_BoneSpace
    translationMode=BMM_Additive
    translationSpace=BCS_BoneSpace
    alphaBoolBlend {blendOption=Linear}
    boneToModify {boneName=dyn_shoulder_pad_l_MSTR}
    translation {y=-3}
   }
   n7 = Get clav_up_l() -> (clav_up_l:double)
   n8 = "Constraint - dyn_shoulder_pad_l_MSTR (P)"!(ComponentPose=@n6, Alpha=@n7, ConstraintWeights_0=1, ConstraintWeights_1:float=0) -> (Pose) {
    alphaBoolBlend {blendOption=Linear}
    boneToModify {boneName=dyn_shoulder_pad_l_MSTR}
    constraintSetup [{targetBone {boneName=upperarm_twist_01_l}}]
    constraintWeights [1,0]
   }
   n9 = "Transform (Modify) Bone - Bone: dyn_shoulder_pad_l_MSTR"(ComponentPose=@n8, Alpha=@n7, Translation="2,-1,1", Rotation="0,0,0", Scale="1,1,1") {
    rotationMode=BMM_Additive
    rotationSpace=BCS_BoneSpace
    translationMode=BMM_Additive
    translationSpace=BCS_BoneSpace
    alphaBoolBlend {blendOption=Linear}
    boneToModify {boneName=dyn_shoulder_pad_l_MSTR}
    translation=<2 -1 1>
   }
   n10 = Get clav_up_r() -> (clav_up_r:double)
   n11 = "Constraint - dyn_shoulder_pad_r_MSTR (P)"!(ComponentPose=@n9, Alpha=@n10, ConstraintWeights_0=1, ConstraintWeights_1:float=0) -> (Pose) {
    alphaBoolBlend {blendOption=Linear}
    boneToModify {boneName=dyn_shoulder_pad_r_MSTR}
    constraintSetup [{targetBone {boneName=upperarm_twist_01_r}}]
    constraintWeights [1,0]
   }
   n12 = "Transform (Modify) Bone - Bone: dyn_shoulder_pad_r_MSTR"(ComponentPose=@n11, Alpha=@n10, Translation="0,0,-1", Rotation="0,0,0", Scale="1,1,1") {
    rotationMode=BMM_Additive
    rotationSpace=BCS_BoneSpace
    translationMode=BMM_Additive
    translationSpace=BCS_BoneSpace
    alphaBoolBlend {blendOption=Linear}
    boneToModify {boneName=dyn_shoulder_pad_r_MSTR}
    translation {z=-1}
   }
   n13 = Get bIsSkyDiving()
   n14 = "Transform (Modify) Bone - Bone: dyn_shoulder_pad_l_MSTR"(ComponentPose=@n12, Alpha="To Float (Boolean)"(InBool=@n13), Translation="0,0,2", Rotation="0,0,0", Scale="1,1,1") {
    rotationMode=BMM_Additive
    rotationSpace=BCS_BoneSpace
    translationMode=BMM_Additive
    translationSpace=BCS_BoneSpace
    alphaBoolBlend {blendOption=Linear}
    boneToModify {boneName=dyn_shoulder_pad_l_MSTR}
    translation {z=2}
   }
   n15 = "Transform (Modify) Bone - Bone: dyn_shoulder_pad_r_MSTR"(ComponentPose=@n14, Alpha="To Float (Boolean)"(InBool=@n13), Translation="0,0,-2", Rotation="0,0,0", Scale="1,1,1") {
    rotationMode=BMM_Additive
    rotationSpace=BCS_BoneSpace
    translationMode=BMM_Additive
    translationSpace=BCS_BoneSpace
    alphaBoolBlend {blendOption=Linear}
    boneToModify {boneName=dyn_shoulder_pad_r_MSTR}
    translation {z=-2}
   }
   n16 = "Constraint - dyn_arm_l_MSTR (PP)"!(ComponentPose=@n15, Alpha=1, ConstraintWeights_0=0.8, ConstraintWeights_1:float=0.2) -> (Pose) {
    alphaBoolBlend {blendOption=Linear}
    boneToModify {boneName=dyn_arm_l_MSTR}
    constraintSetup [{targetBone {boneName=spine_05}}, {targetBone {boneName=clavicle_l}}]
    constraintWeights [1,1]
   }
   n17 = "Constraint - dyn_arm_r_MSTR (PP)"!(ComponentPose=@n16, Alpha=1, ConstraintWeights_0=0.8, ConstraintWeights_1:float=0.2) -> (Pose) {
    alphaBoolBlend {blendOption=Linear}
    boneToModify {boneName=dyn_arm_r_MSTR}
    constraintSetup [{targetBone {boneName=spine_05}}, {targetBone {boneName=clavicle_r}}]
    constraintWeights [1,1]
   }
   n18 = Control Rig(Source=Component To Local(ComponentPose=@n17)) {
    alphaBoolBlend {blendOption=Linear}
    controlRigAssetReference {blueprintRigClass=&/F_MED_Assemble_R_CTRL.F_MED_Assemble_R_CTRL_C}
   }
   n19 = Get arm_up() -> (arm_up:double)
   n20 = "Transform (Modify) Bone - Bone: dyn_arm_aim_r"(ComponentPose=Local To Component(LocalPose=@n18), Alpha=@n19, Translation="2,0,0", Rotation="-30,-10,-10", Scale="1,1,1") {
    rotationMode=BMM_Additive
    rotationSpace=BCS_BoneSpace
    translationMode=BMM_Additive
    translationSpace=BCS_BoneSpace
    alphaBoolBlend {blendOption=Linear}
    boneToModify {boneName=dyn_arm_aim_r}
    rotation=<rot -30 -10 -10>
    translation {x=2}
   }
   n21 = "Transform (Modify) Bone - Bone: dyn_arm_aim_l"(ComponentPose=@n20, Alpha=@n19, Translation="3.5,0,0", Rotation="-10,-10,2", Scale="1,1,1") {
    rotationMode=BMM_Additive
    rotationSpace=BCS_BoneSpace
    translationMode=BMM_Additive
    translationSpace=BCS_BoneSpace
    alphaBoolBlend {blendOption=Linear}
    boneToModify {boneName=dyn_arm_aim_l}
    rotation=<rot -10 -10 2>
    translation {x=3.5}
   }
   n22 = "Transform (Modify) Bone - Bone: dyn_arm_aim_r"(ComponentPose=@n21, Alpha=1, Translation="0.5,0,1", Rotation="0,0,0", Scale="1,1,1") {
    rotationMode=BMM_Additive
    rotationSpace=BCS_BoneSpace
    translationMode=BMM_Additive
    translationSpace=BCS_BoneSpace
    alphaBoolBlend {blendOption=Linear}
    boneToModify {boneName=dyn_arm_aim_r}
    translation {x=0.5, z=1}
   }
   n23 = "Transform (Modify) Bone - Bone: dyn_arm_aim_l"(ComponentPose=@n22, Alpha=1, Translation="0.5,0,1", Rotation="0,0,0", Scale="1,1,1") {
    rotationMode=BMM_Additive
    rotationSpace=BCS_BoneSpace
    translationMode=BMM_Additive
    translationSpace=BCS_BoneSpace
    alphaBoolBlend {blendOption=Linear}
    boneToModify {boneName=dyn_arm_aim_l}
    translation {x=0.5, z=1}
   }
   n24 = "Transform (Modify) Bone - Bone: dyn_arm_aim_r"(ComponentPose=@n23, Alpha=Get clav_up_r_plus() -> (clav_up_r_plus:double), Translation="3,0,1", Rotation="0,0,0", Scale="1,1,1") {
    rotationMode=BMM_Additive
    rotationSpace=BCS_BoneSpace
    translationMode=BMM_Additive
    translationSpace=BCS_BoneSpace
    alphaBoolBlend {blendOption=Linear}
    boneToModify {boneName=dyn_arm_aim_r}
    translation {x=3, z=1}
   }
   n25 = "Transform (Modify) Bone - Bone: dyn_arm_aim_l"(ComponentPose=@n24, Alpha=Get clav_up_plus(), Translation="1,2,0", Rotation="0,0,0", Scale="1,1,1") {
    rotationMode=BMM_Additive
    rotationSpace=BCS_BoneSpace
    translationMode=BMM_Additive
    translationSpace=BCS_BoneSpace
    alphaBoolBlend {blendOption=Linear}
    boneToModify {boneName=dyn_arm_aim_l}
    translation=<1 2>
   }
   Output Pose(Result=Component To Local(ComponentPose=@n25)) {layerGroup=None}
   Get clav_up_r()
   Get clav_up_l()
  ]}
 EventGraph {nodes [
   Event BlueprintUpdateAnimation()
   > Set clav_up_r(clav_up_r:double=GetAlphaFromReferenceTransformDelta(
     SkeletalMeshComponent=Get Body_Mesh()
     BoneName=upperarm_r
     BaseName=clavicle_r
     DeltaComponentType=QuaternionTwist
     InputRange=(X=0,Y=-120)
     TwistAxis="0,0,1"
    )) -> (Output_Get:double)
   > Set clav_up_l(clav_up_l:double=GetAlphaFromReferenceTransformDelta(
     SkeletalMeshComponent=Get Body_Mesh()
     BoneName=upperarm_l
     BaseName=clavicle_l
     DeltaComponentType=QuaternionTwist
     InputRange=(X=-10,Y=-40)
     TwistAxis="0,0,1"
    )) -> (Output_Get:double)
   > Set clav_up_plus(clav_up_plus:double=GetAlphaFromReferenceTransformDelta(
     SkeletalMeshComponent=Get Body_Mesh()
     BoneName=clavicle_l
     BaseName=spine_05
     DeltaComponentType=EulerY
     InputRange=(X=13,Y=44)
     TwistAxis="0,0,1"
    )) -> (Output_Get:double)
   > Set arm_up(arm_up:double=GetAlphaFromReferenceTransformDelta(
     SkeletalMeshComponent=Get Body_Mesh()
     BoneName=clavicle_l
     BaseName=spine_05
     DeltaComponentType=EulerY
     InputRange=(X=4,Y=43)
     TwistAxis="0,0,1"
    )) -> (Output_Get:double)
   > Set clav_up_r_plus(clav_up_r_plus:double=GetAlphaFromReferenceTransformDelta(
     SkeletalMeshComponent=Get Body_Mesh()
     BoneName=clavicle_r
     BaseName=spine_05
     DeltaComponentType=EulerY
     InputRange=(X=14,Y=35)
     TwistAxis="0,0,1"
    )) -> (Output_Get:double)
   Event BlueprintInitializeAnimation()
   > Set Body_Mesh(Body_Mesh=GetPartSkeletalMeshComponent())
  ]}
]
variables [
 arm_up {type=number}
 Body_Mesh {type=SkeletalMeshComponent}
 clav_up_l {type=number}
 clav_up_plus {type=number}
 clav_up_r {type=number}
 clav_up_r_plus {type=number}
]
