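# Control Rig for parts of the medium female player body Assemble_R that keeps the dynamic head and neck bones pointed at each other.
# The solve first aims dyn_head's X axis at dyn_neck, then aims dyn_neck back at dyn_head with partial weight.
# Both aims have a zero secondary axis.
# The hierarchy also holds spine, face, scope and wire bones, which the graph leaves untouched.
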ControlRigBlueprint /Game/Characters/Player/Female/Medium/Bodies/F_MED_Assemble_R/Meshes/Parts/F_MED_Assemble_R_CTR_RIG
graphs [
 RigVMFunctionLibrary {links [], nodes []}
 RigVMModel {nodes [
   RigUnit_BeginExecution = &RigUnit_BeginExecution()
   /Script/ControlRig.RigUnit_AimItem(
    DebugSettings=(Scale=10,WorldOffset=(Rotation=(X=0,Y=0,Z=0,W=1),Translation=(X=0,Y=0,Z=0),Scale3D=(X=1,Y=1,Z=1)))
    Item=(Type=Bone,Name="dyn_head")
    Primary=(Weight=1,Axis=(X=1,Y=0,Z=0),Target=(X=0,Y=0,Z=0),Kind=Location,Space=(Type=Bone,Name="dyn_neck"))
    Secondary=(Weight=1,Axis=(X=0,Y=0,Z=0),Target=(X=0,Y=0,Z=0),Kind=Location,Space=(Type=Bone))
    Weight=1
    ExecutePin=@RigUnit_BeginExecution.ExecutePin
   ) {is=struct RigVMUnitNode}
   > &RigUnit_AimItem(
    DebugSettings=(Scale=10,WorldOffset=(Rotation=(X=0,Y=0,Z=0,W=1),Translation=(X=0,Y=0,Z=0),Scale3D=(X=1,Y=1,Z=1)))
    Item=(Type=Bone,Name="dyn_neck")
    Primary=(Weight=1,Axis=(X=1,Y=0,Z=0),Target=(X=0,Y=0,Z=0),Kind=Location,Space=(Type=Bone,Name="dyn_head"))
    Secondary=(Weight=1,Axis=(X=0,Y=0,Z=0),Target=(X=0,Y=0,Z=0),Kind=Location,Space=(Type=Bone))
    Weight=0.7
   )
  ]}
]
hierarchy [
 root {loc=<0 0 0>, rot=<0 0 0 1>}
 pelvis {parent=root, loc=<0 -0.5301 98.0571>, rot=<0.0127 0.707 -0.0127 -0.707>}
 spine_01 {parent=pelvis, loc=<6.1208 -0.9328 0>, rot=<0 0 0.034 -0.9994>}
 spine_02 {parent=spine_01, loc=<6.5996 -0.2095 0>, rot=<0 0 0.0238 -0.9997>}
 spine_03 {parent=spine_02, loc=<6.4914 -0.1237 0>, rot=<0 0 0.0149 -0.9999>}
 spine_04 {parent=spine_03, loc=<6.9598 0.0794 0>, rot=<0 0 -0.0132 -0.9999>}
 spine_05 {parent=spine_04, loc=<6.9987 0 0>, rot=<0 0 -0.0015 -1>}
 neck_01 {parent=spine_05, loc=<11.3138 0.1529 0>, rot=<0 0 -0.2003 -0.9797>}
 neck_02 {parent=neck_01, loc=<4.5627 0 0>, rot=<0 0 0.0504 -0.9987>}
 head {parent=neck_02, loc=<5.3914 0 0>, rot=<0 0 0.115 -0.9934>}
 faceAttach {parent=head, loc=<0.6201 0.9296 0>, rot=<0.0082 -0.7071 -0.0082 -0.7071>, scale=<0.89 0.89 0.89>}
 R_brow_mid {parent=faceAttach, loc=<-3.2828 8.4055 5.3916>, rot=<-0.0001 0.0042 -0.2553 -0.9669>}
 R_eye {parent=faceAttach, loc=<-3.372 7.1482 3.2788>, rot=<0 0 0.9998 -0.0187>}
 R_eye_lid_upper_mid {parent=R_eye, loc=<-0.0396 -1.1714 0.2631>, rot=<0.0523 -0.0009 0.0167 -0.9985>}
 R_eye_lid_lower_mid {parent=R_eye, loc=<-0.0085 -0.9425 -0.7051>, rot=<0.0329 -0.0005 0.0167 -0.9993>}
 dyn_scope_MSTR {parent=head, loc=<5.0689 13.5502 0.0016>, rot=<0.0029 -0.7071 -0.0029 -0.7071>}
 dyn_focus {parent=dyn_scope_MSTR, loc=<0.0016 3.2971 -0.293>, rot=<0 0 0 1>}
 dyn_lens {parent=dyn_focus, loc=<-0.0008 -0.2018 -0.0636>, rot=<0 0 0 1>}
 dyn_wire_back {parent=head, loc=<9.5344 -8.3177 0.6942>, rot=<0.0029 -0.7071 -0.0029 -0.7071>}
 dyn_head {parent=head, loc=<-4.9119 1.3716 0.0004>, rot=<0.0957 -0.7006 -0.7006 0.0956>}
 dyn_neck {parent=spine_05, loc=<10.959 2.8834 -0.0005>, rot=<0.6967 0.1209 -0.1207 -0.6967>}
]
