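# Physics asset for a part mesh of the medium female player body F_MED_Assemble_R, covering a dangling back wire and the head.
# It holds one small capsule body on dyn_wire_back and a larger capsule on the head.
# A single constraint joins the wire bone to the head, with limited swing and a very tight twist limit.
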PhysicsAsset /Game/Characters/Player/Female/Medium/Bodies/F_MED_Assemble_R/Meshes/Parts/F_MED_Assemble_R_Physics
bodies {
 dyn_wire_back [{shapeType=Capsule, length=0.505, radius=0.505, center=<-0.1954 -0.2721 -1.6693>, rotation=<rot 0 -94.477 -82.6628>}]
 head [{shapeType=Capsule, length=4.8906, radius=7.5839, center=<8.3458 -0.2972 0>, rotation=<rot 0 -11.0127 0>}]
}
constraints [{bone1Name=dyn_wire_back, bone2Name=head, swing1LimitDegrees=45, swing2LimitDegrees=45, twistLimitDegrees=2}]
