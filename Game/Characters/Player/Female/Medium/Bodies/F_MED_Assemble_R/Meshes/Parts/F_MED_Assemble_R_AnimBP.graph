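# Animation Blueprint for the F_MED_Assemble_R player body part, animating a lens focus bone and secondary physics.
# It copies the attached parent's pose, then additively rotates and shifts dyn_focus using alphas read each update from right eye and upper eyelid twist relative to the head on Body_Mesh.
# A RigidBody node in root space follows, its strength switched by bIsTargeting, before a Control Rig finishes the pose.
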AnimBlueprint /Game/Characters/Player/Female/Medium/Bodies/F_MED_Assemble_R/Meshes/Parts/F_MED_Assemble_R_AnimBP
target_skeleton=&/F_MED_Assemble_R_Skeleton
graphs [
 AnimGraph {nodes [
   Component To Local(ComponentPose=Local To Component())
   Component To Local()
   Component To Local()
   Component To Local()
   n1 = "Transform (Modify) Bone - Bone: dyn_focus"(
    ComponentPose=Local To Component(LocalPose=Copy Pose From Mesh() {bUseAttachedParent=true})
    Alpha=Get lens_rot() -> (lens_rot:double)
    Translation="0,0,0"
    Rotation="60,0,0"
    Scale="1,1,1"
   ) {
    rotationMode=BMM_Additive
    rotationSpace=BCS_BoneSpace
    translationMode=BMM_Additive
    translationSpace=BCS_BoneSpace
    alphaBoolBlend {blendOption=Linear}
    boneToModify {boneName=dyn_focus}
    rotation {pitch=60}
   }
   n2 = "Transform (Modify) Bone - Bone: dyn_focus"(ComponentPose=@n1, Alpha=Get lens_mov() -> (lens_mov:double), Translation="0,0.25,0", Rotation="0,0,0", Scale="1,1,1") {
    rotationMode=BMM_Additive
    rotationSpace=BCS_BoneSpace
    translationMode=BMM_Additive
    translationSpace=BCS_BoneSpace
    alphaBoolBlend {blendOption=Linear}
    boneToModify {boneName=dyn_focus}
    translation {y=0.25}
   }
   n3 = RigidBody(
    ComponentPose=Local To Component(LocalPose=Component To Local(ComponentPose=@n2))
    Alpha=SelectFloat(A=0.6, B=0.8, bPickA=Get bIsTargeting())
   ) {
    bClampLinearTranslationLimitToRefPose=true
    bOverrideWorldGravity=true
    overridePhysicsAsset=&/F_MED_Assemble_R_Physics
    simulationSpace=BaseBoneSpace
    alphaBoolBlend {blendOption=Linear}
    baseBoneRef {boneName=root}
    componentLinearAccScale=<0.4 0.4 0.5>
    componentLinearVelScale=<0.2 0.2 0.25>
    overrideWorldGravity {z=-980}
   }
   n4 = Control Rig(Source=Component To Local(ComponentPose=@n3)) {
    alphaBoolBlend {blendOption=Linear}
    controlRigAssetReference {blueprintRigClass=&/F_MED_Assemble_R_CTR_RIG.F_MED_Assemble_R_CTR_RIG_C}
   }
   Output Pose(Result=@n4) {layerGroup=None}
  ]}
 EventGraph {nodes [
   Event BlueprintUpdateAnimation()
   > Set lens_rot(lens_rot:double=GetAlphaFromReferenceTransformDelta(
     SkeletalMeshComponent=Get Body_Mesh()
     BoneName=R_eye
     BaseName=head
     DeltaComponentType=QuaternionTwist
     InputRange=(X=0,Y=50)
     TwistAxis="0,0,1"
    )) -> (Output_Get:double)
   > Set lens_mov(lens_mov:double=GetAlphaFromReferenceTransformDelta(
     SkeletalMeshComponent=Get Body_Mesh()
     BoneName=R_eye_lid_upper_mid
     BaseName=head
     DeltaComponentType=QuaternionTwist
     InputRange=(X=0,Y=-100)
     TwistAxis="0,0,1"
    )) -> (Output_Get:double)
   Event BlueprintInitializeAnimation()
   > Set Body_Mesh(Body_Mesh=GetPartSkeletalMeshComponent())
   Delay(Duration=5)
   GetAlphaFromReferenceTransformDelta(BoneName=R_eye_lid_upper_mid, BaseName=head, DeltaComponentType=QuaternionTwist, InputRange=(X=0,Y=-120), TwistAxis="0,0,1")
  ]}
]
variables [
 Body_Mesh {type=SkeletalMeshComponent}
 lens_mov {type=number}
 lens_rot {type=number}
]
